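# Physics asset for the SharkFry medium male player body, providing capsule bodies for dangling left and right laces.
# Short lace chains (one dynamic segment left, two right) hang from small base capsules, with tight swing limits and zero-degree twist limits.
# Pelvis and thigh capsules give body volume around the laces, and the thighs and lace bases link to the pelvis through fully free angular constraints.
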PhysicsAsset /BRCosmetics/Characters/Player/Male/Medium/Bodies/M_MED_SharkFry/Meshes/M_MED_SharkFry_Physics
bodies {
 Base_laces_L [{shapeType=Capsule, length=0.505, radius=0.505, shapeName=Base_laces_L_capsule, rotation=<rot 0 -90 0>}]
 Base_laces_R [{shapeType=Capsule, length=0.505, radius=0.505, shapeName=Base_laces_R_capsule, rotation=<rot 0 -90 0>}]
 Dyn_laces_L_01 [{shapeType=Capsule, length=5.051, radius=1.7176, shapeName=Dyn_laces_L_01_capsule, center=<5.1 0 0>, rotation=<rot 90 0 0>}]
 Dyn_laces_R_01 [{shapeType=Capsule, length=4.3534, radius=2.146, shapeName=Dyn_laces_R_01_capsule, center=<5.6681 0 0>, rotation=<rot 90 179.156 179.156>}]
 Dyn_laces_R_02 [{shapeType=Capsule, length=6.2713, radius=1.9677, shapeName=Dyn_laces_R_02_capsule, center=<5.3 0 0>, rotation=<rot 90 -144.7356 -144.7356>}]
 pelvis [
  {shapeType=Capsule, length=10.77, radius=9.23, shapeName=pelvis_capsule, center=<2.1125 -0.9324 0>}
  {shapeType=Capsule, length=9.2346, radius=8.7654, shapeName=pelvis_capsule, center=<-1.5999 0.473 0>}
 ]
 thigh_l [{shapeType=Capsule, length=25.54, radius=7.7209, shapeName=thigh_l_capsule, center=<-18.7541 0 0>, rotation=<rot 90 0 0>}]
 thigh_r [{shapeType=Capsule, length=25.7817, radius=8.1016, shapeName=thigh_r_capsule, center=<18.8012 0 0>, rotation=<rot 90 0 0>}]
}
constraints [
 {bone1Name=Dyn_laces_R_02, bone2Name=Dyn_laces_R_01, swing1LimitDegrees=20, swing2LimitDegrees=20, twistLimitDegrees=0}
 {bone1Name=Dyn_laces_L_01, bone2Name=Base_laces_L, swing1LimitDegrees=25, swing2LimitDegrees=35, twistLimitDegrees=0}
 {bone1Name=Dyn_laces_R_01, bone2Name=Base_laces_R, swing1LimitDegrees=25, swing2LimitDegrees=35, twistLimitDegrees=0}
 {
  bone1Name=thigh_l
  bone2Name=pelvis
  swing1LimitDegrees=45
  swing1Motion=Free
  swing2LimitDegrees=45
  swing2Motion=Free
  twistLimitDegrees=45
  twistMotion=Free
 }
 {
  bone1Name=thigh_r
  bone2Name=pelvis
  swing1LimitDegrees=45
  swing1Motion=Free
  swing2LimitDegrees=45
  swing2Motion=Free
  twistLimitDegrees=45
  twistMotion=Free
 }
 {
  bone1Name=Base_laces_R
  bone2Name=pelvis
  swing1LimitDegrees=45
  swing1Motion=Free
  swing2LimitDegrees=45
  swing2Motion=Free
  twistLimitDegrees=45
  twistMotion=Free
 }
 {
  bone1Name=Base_laces_L
  bone2Name=pelvis
  swing1LimitDegrees=45
  swing1Motion=Free
  swing2LimitDegrees=45
  swing2Motion=Free
  twistLimitDegrees=45
  twistMotion=Free
 }
]
